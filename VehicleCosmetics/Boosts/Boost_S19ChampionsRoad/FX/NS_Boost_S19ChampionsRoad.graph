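# Niagara system for the Season 19 Champions Road vehicle boost cosmetic, drawing layered ribbon trails from the vehicle's boost sockets.
# Blue and Golden emitters spawn ribbons per port at skeletal positions, with backward velocity and fading inherited source movement.
# Curves shape width and color, while spawn-group arrays vary dynamic material parameters between ribbons.
# The frontend preview flag changes velocity and Golden's color and material values; painted-variant overrides are disabled.
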NiagaraSystem /VehicleCosmetics/Boosts/Boost_S19ChampionsRoad/FX/NS_Boost_S19ChampionsRoad
dependencies {
 usedDataInterfaces [
  &NiagaraDataInterfaceCurve
  /Script/Niagara.NiagaraDataInterfaceArrayBool
  &NiagaraDataInterfaceArrayFloat
  &NiagaraDataInterfaceVectorCurve
  &NiagaraDataInterfaceSkeletalMesh
 ]
 usedDynamicInputs []
 usedModules [
  &self:SystemSpawnScript.SystemScriptSource.SystemScriptGraph.NiagaraNodeAssignment_0.SetVariables_#1
  &SystemState
 ]
}
emitters {
 NE_Ribbon_Blue {
  emitterSpawnScript [/SetVariables(Emitter.bIsInGame:NiagaraBool=&Boolean_NotOperation(Bool=$User.bIsFrontendPreview)) {isSetParameters=true}]
  emitterUpdateScript [
   &EmitterState()
   &VC_FilteredSocketCount()
   &VC_RateQualityHandler(Epic Spawn Rate=45, Low Spawn Rate=45, Medium Spawn Rate=45, Use Custom Rate?=true)
   &VC_SpawnRibbon_v1(Number of Ribbons=$Emitter.PortCount, Spawn Time=-1, SpawnRate=$Emitter.Spawn Rate)
  ]
  particleSpawnScript [
   &VC_PaintedVariantsOverride() {enabled=false}
   &VC_InitializeRibbon_v1(Enable Skeletal Position=true, Lifetime=0.95, Ribbon Width=50)
   &AddVelocity(Velocity=&MakeCustomVectorFromBool(Bool=$User.bIsFrontendPreview, True Vector=<-1500 0 0>, False Vector=<-1000 0 0>)) {module=AddVelocity001}
   /SetVariables(Particles.R2_Color:LinearColor=<rgba 0.2875 0.2875 1.15 1>) {isSetParameters=true}
  ]
  particleUpdateScript [
   &ParticleState()
   &InheritSourceMovement(
    Applied Position Scale=&VectorFromFloat(Value=&FloatFromCurve(FloatCurve=DataInterface() {propertyValues {Curve {keys [
          {interpMode=RCIM_Cubic, tangentMode=RCTM_Break, time=0, value=1}
          {arriveTangent=-2.9025, interpMode=RCIM_Cubic, tangentMode=RCTM_Break, time=1, value=0}
         ]}}}))
    ApplyForce=false
    ApplyPosition=true
    ApplyPositionFalloff=false
   )
   &SolveForcesAndVelocity(Clamp Velocity=false, Limit Acceleration=false)
   &VC_UpdateRibbon_v1(Enable Skeletal Position=true)
   &DynamicMaterialParameters(
    Index 2 Param 1=1
    Index 2 Param 2=&MakeCustomFloatFromBool(
     Bool=/Niagara/DynamicInputs/Arrays/SelectBoolFromArray(
      Array Sampling Mode:ENiagara_LifetimeMode=::NewEnumerator0
      Direct Array Index:NiagaraInt32=$Emitter.PortCount
      Direct Array Mode:ENiagara_WrapClamp=::NewEnumerator0
      Evaluation Type:ENiagaraRandomnessEvaluation=::NewEnumerator0
      Recalculate Random Each Loop:NiagaraBool=false
      Randomness Mode:ENiagaraRandomnessMode=::NewEnumerator0
      Random Seed:NiagaraInt32=?
      Fixed Random Seed:NiagaraBool=false
      Bool Selection Array:NiagaraDataInterfaceArrayBool=DataInterface() {propertyValues {GpuSyncMode=SyncCpuToGpu, MaxElements=0, BoolData [true,true,false]}}
     )
     True Float=1
     False Float=&SelectFloatFromArray(Direct Array Index=&SpawnGroup(), Float Selection Array=DataInterface() {propertyValues {FloatData [-1,1]}})
    )
    Index 2 Param 3=1
    Index 2 Param 4=1
    Index 3 Param 1=1
    Index 3 Param 2=1
    Module.Index0Float0Write=true
    Module.Index0Float1Write=true
    Module.Index0Float2Write=true
    Module.Index0Float3Write=true
    Module.Index1Float0Write=true
    Module.Index1Float1Write=true
    Module.Index1Float2Write=true
    Module.Index1Float3Write=true
    Module.Index2Float0Write=true
    Module.Index2Float1Write=true
    Module.Index2Float2Write=true
    Module.Index2Float3Write=true
    Module.Index3Float0Write=true
    Module.Index3Float1Write=true
    Module.Index3Float2Write=true
    Module.Index3Float3Write=true
    Write Parameter Index 1=::NewEnumerator1
    Write Parameter Index 2=::NewEnumerator1
    Write Parameter Index 3=::NewEnumerator1
   )
   &ScaleRibbonWidth(Ribbon Width Scale=&FloatFromCurve(FloatCurve=DataInterface() {propertyValues {Curve {keys [<at 0.5 1>,<at 1 0.4>]}}}))
   &ScaleColor(
    Scale Alpha=&FloatFromCurve(
     FloatCurve=DataInterface() {propertyValues {Curve {keys [
         <at 0 0 RCIM_Cubic>
         {arriveTangent=1.4286, interpMode=RCIM_Cubic, leaveTangent=1.4286, time=0.15, value=1.0238}
         {arriveTangent=-1.5087, interpMode=RCIM_Cubic, leaveTangent=-1.5087, tangentMode=RCTM_User, time=0.7, value=1}
         {arriveTangent=-0.1935, tangentMode=RCTM_User, time=1, value=0}
        ]}}}
     Scale Curve=0.4
    )
    Scale RGB=&VectorFromCurve(
     VectorCurve=DataInterface() {propertyValues {
       XCurve {keys [
         <at 0 0.15 RCIM_Cubic>
         {arriveTangent=-0.2167, interpMode=RCIM_Cubic, leaveTangent=-0.2167, time=0.2, value=0.06}
         {arriveTangent=-0.0833, interpMode=RCIM_Cubic, leaveTangent=-0.0833, time=0.6, value=0.02}
         <at 0.8 0.01 RCIM_Cubic>
        ]}
       YCurve {keys [
         <at 0 0.4 RCIM_Cubic>
         {arriveTangent=-0.5667, interpMode=RCIM_Cubic, leaveTangent=-0.5667, time=0.2, value=0.2}
         {arriveTangent=-0.3167, interpMode=RCIM_Cubic, leaveTangent=-0.3167, time=0.6, value=0.06}
         <at 0.8 0.01 RCIM_Cubic>
        ]}
       ZCurve {keys [
         <at 0 0.95 RCIM_Cubic>
         {arriveTangent=-1.0833, interpMode=RCIM_Cubic, leaveTangent=-1.0833, time=0.2, value=1.2}
         {arriveTangent=-1.8167, interpMode=RCIM_Cubic, leaveTangent=-1.8167, time=0.6, value=0.3}
         <at 0.8 0.11 RCIM_Cubic>
        ]}
      }}
     Scale Curve=&Multiply_VectorByFloat(Vector=<1.4 1 1>, Float=0.75)
    )
    ScaleA=true
    ScaleRGB=true
   )
   /SetVariables(Particles.R2_Color_Scaled:LinearColor=&VC_DynamicScaleColor(
     Color to Scale:LinearColor=$Particles.R2_Color
     RGB Scale:Vector3f=<10 10 10>
     Alpha Scale:NiagaraFloat=&FloatFromCurve(
      FloatCurve=DataInterface() {propertyValues {Curve {keys [
          <at 0 0 RCIM_Cubic>
          {arriveTangent=1.4286, interpMode=RCIM_Cubic, leaveTangent=1.4286, time=0.15, value=1.0238}
          {arriveTangent=-1.2045, interpMode=RCIM_Cubic, leaveTangent=-1.2045, time=0.7, value=1}
          {arriveTangent=-0.1935, tangentMode=RCTM_User, time=1, value=0}
         ]}}}
      Scale Curve=0.55
     )
    )) {isSetParameters=true}
  ]
  rendererClasses [&NiagaraRibbonRendererProperties]
  renderers [
   {rendererClass=&NiagaraRibbonRendererProperties}
   {rendererClass=&NiagaraRibbonRendererProperties, rendererIndex=1}
   {rendererClass=&NiagaraRibbonRendererProperties, rendererIndex=2}
   {rendererClass=&NiagaraRibbonRendererProperties, rendererIndex=3}
  ]
 }
 NE_Ribbon_Golden {
  emitterSpawnScript [/SetVariables(Emitter.bIsInGame:NiagaraBool=&Boolean_NotOperation(Bool=$User.bIsFrontendPreview)) {isSetParameters=true}]
  emitterUpdateScript [
   &EmitterState()
   &VC_FilteredSocketCount()
   &VC_RateQualityHandler(Epic Spawn Rate=30, High Spawn Rate=20, Low Spawn Rate=20, Medium Spawn Rate=20, Use Custom Rate?=true)
   &VC_SpawnRibbon_v1(Number of Ribbons=$Emitter.PortCount, Spawn Time=-1, SpawnRate=60)
  ]
  particleSpawnScript [
   &VC_PaintedVariantsOverride() {enabled=false}
   &VC_InitializeRibbon_v1(Enable Skeletal Position=true, Lifetime=0.95, Offset Amount=10, Offset Axis:ENiagaraOrientationAxis=::XAxis, Ribbon Width=30)
   &AddVelocity(Velocity=&MakeCustomVectorFromBool(Bool=$User.bIsFrontendPreview, True Vector=<-1500 0 0>, False Vector=<-1000 0 0>)) {module=AddVelocity001}
  ]
  particleUpdateScript [
   &ParticleState()
   &InheritSourceMovement(
    Applied Position Scale=&VectorFromFloat(Value=&FloatFromCurve(FloatCurve=DataInterface() {propertyValues {Curve {keys [
          {interpMode=RCIM_Cubic, tangentMode=RCTM_Break, time=0, value=1}
          {arriveTangent=-2.9025, interpMode=RCIM_Cubic, tangentMode=RCTM_Break, time=1, value=0}
         ]}}}))
    ApplyForce=false
    ApplyPosition=true
    ApplyPositionFalloff=false
   )
   &SolveForcesAndVelocity(Clamp Velocity=false, Limit Acceleration=false)
   &ScaleRibbonWidth(Ribbon Width Scale=&FloatFromCurve(FloatCurve=DataInterface() {propertyValues {Curve {keys [<at 0.5 1>,<at 1 0.5>]}}}))
   &ScaleColor(
    Scale Alpha=&FloatFromCurve(
     FloatCurve=DataInterface() {propertyValues {Curve {keys [
         <at 0.0365 0.3 RCIM_Cubic>
         {arriveTangentWeight=0.1966, interpMode=RCIM_Cubic, leaveTangentWeight=0.1408, tangentMode=RCTM_User, time=0.6606, value=0.368}
         <at 1 0.3717 RCIM_Cubic>
        ]}}}
     Scale Curve=&SelectFloatFromArray(
      Direct Array Index=&SpawnGroup()
      Float Selection Array=DataInterface() {propertyValues {FloatData [1.25,1.25]}}
     )
    )
    Scale RGB=&VectorFromCurve(
     VectorCurve=DataInterface() {propertyValues {
       XCurve {defaultValue=7, keys [<at 0 3>,<at 0.2 7>]}
       YCurve {defaultValue=7, keys [<at 0 3>,<at 0.2 7>]}
       ZCurve {defaultValue=7, keys [<at 0 3>,<at 0.2 7>]}
      }}
     Scale Curve=<0 5 5>
    )
    ScaleA=true
    ScaleRGB=true
   ) {enabled=false}
   &ScaleColor(
    Scale Alpha=&FloatFromCurve(
     FloatCurve=DataInterface() {propertyValues {Curve {keys [<at 0 0.3>]}}}
     Scale Curve=&SelectFloatFromArray(
      Direct Array Index=&SpawnGroup()
      Float Selection Array=DataInterface() {propertyValues {FloatData [1.25,-1.25]}}
     )
    )
    Scale RGB=&VectorFromCurve(
     VectorCurve=DataInterface() {propertyValues {
       XCurve {defaultValue=7, keys [<at 0 0.5>,<at 0.2 6>]}
       YCurve {defaultValue=7, keys [<at 0 0.5>,<at 0.2 6>]}
       ZCurve {defaultValue=7, keys [<at 0 0.5>,<at 0.2 6>]}
      }}
     Scale Curve=&MakeCustomVectorFromBool(Bool=$User.bIsFrontendPreview, True Vector=<0 4 4>, False Vector=<0 2.5 1.5>)
    )
    ScaleA=true
    ScaleRGB=true
   ) {module=ScaleColor001}
   &DynamicMaterialParameters(
    Index 0 Param 1=&MakeCustomFloatFromBool(Bool=$User.bIsFrontendPreview, True Float=1, False Float=0.65)
    Index 0 Param 2=&SelectFloatFromArray(Direct Array Index=&SpawnGroup(), Float Selection Array=DataInterface() {propertyValues {FloatData [-7,7]}})
    Index 0 Param 3=&SelectFloatFromArray(Direct Array Index=&SpawnGroup(), Float Selection Array=DataInterface() {propertyValues {FloatData [1.5,-1.5]}})
    Index 0 Param 4=&SelectFloatFromArray(Direct Array Index=&SpawnGroup(), Float Selection Array=DataInterface() {propertyValues {FloatData [-0.5,-1]}})
    Module.Index0Float0Write=true
    Module.Index0Float1Write=true
    Module.Index0Float2Write=true
    Module.Index0Float3Write=true
   )
   &VC_UpdateRibbon_v1(Enable Skeletal Position=true)
  ]
  rendererClasses [&NiagaraRibbonRendererProperties]
  renderers [
   {rendererClass=&NiagaraRibbonRendererProperties}
   {rendererClass=&NiagaraRibbonRendererProperties, rendererIndex=1}
  ]
 }
}
summary {
 systemName=NS_Boost_S19ChampionsRoad
 emitters [
  {emitterName=NE_Ribbon_Blue, rendererClasses [&NiagaraRibbonRendererProperties]}
  {emitterName=NE_Ribbon_Golden, rendererClasses [&NiagaraRibbonRendererProperties]}
 ]
 userVariables [
  User.bIsFrontendPreview {type=NiagaraBool}
  User.bIsPainted {type=NiagaraBool}
  User.ColorVariantID {type=NiagaraInt32}
  User.PaintedColor {type=LinearColor, description="", defaultValue {struct=LinearColor, value=<rgba 1 1 1 1>}}
  User.PaintedScalar {type=NiagaraFloat, defaultValue=1, description=""}
  User.PortCount {type=NiagaraInt32, defaultValue=2}
 ]
}
